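# Material function in the Pivot Painter 2 folder, exposed under Math, that converts a flat index into UVs for a 2D array of given dimensions.
# Index modulo width gives the column and the floored quotient gives the row.
# A half-cell offset centers the position before division by the dimensions normalizes it.
# As wired, both components of the UVs output take the R (column) result, leaving the row unused.
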MaterialFunction /Engine/Functions/Engine_MaterialFunctions02/PivotPainter2/2dArrayLookupByIndex
graph {
 properties {bExposeToLibrary=true, description=Finds an index's location in a 2d Array., libraryCategoriesText [Math]}
 nodes [
  n1 = FunctionInput() {inputName=Index, inputType=FunctionInput_Scalar}
  n2 = FunctionInput() {inputName=Array Dimensions, inputType=FunctionInput_Vector2}
  n3 = MaterialFunctionCall(Float2=@n2) {materialFunction=&BreakOutFloat2Components}
  n4 = AppendVector(A=Fmod(A=@n1, B=@n3.R), B=Floor(None=Divide(A=@n1, B=@n3.R)))
  n5 = MaterialFunctionCall(Float2=Divide(A=Add(A=@n4) {constB=0.5}, B=@n2)) {materialFunction=&BreakOutFloat2Components}
  FunctionOutput(None=MaterialFunctionCall(X=@n5.R, Y=@n5.R) {materialFunction=&MakeFloat2}.Result) {outputName=UVs}
 ]
}
