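# Material function in the Pivot Painter 2 library that decodes an 8-bit alpha axis extent value from a Pivot Painter 2 texture into world-space length.
# The scalar input is multiplied by a constant to expand it to the 2048-unit range, then clamped only at a minimum of 8.
# The result represents the model's length along an axis from its pivot point, nominally in increments of 8.
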MaterialFunction /Engine/Functions/Engine_MaterialFunctions02/PivotPainter2/ms_PivotPainter2_Decode8BitAlphaAxisExtent
graph {
 properties {
  bExposeToLibrary=true
  description=Rescale 8 bit axis extent texture data information from Pivot Painter 2 Into world space data.
  libraryCategoriesText [Pivot Painter 2]
 }
 nodes [
  n1 = FunctionInput() {
   bUsePreviewValueAsDefault=true
   desc="Insert a pivot painter 2 alpha texture component from a texture with a 8 bit alpha extent value. This can be generated by selecting the appropriate option from the alpha output drop down options in the MAXScript. "
   description="Insert a pivot painter 2 alpha texture component from a texture with a 8 bit alpha extent value. This can be generated by selecting the appropriate option from the alpha output drop down options in the MAXScript. "
   inputName=8 Bit Alpha Extent Value
   inputType=FunctionInput_Scalar
   sortPriority=1
   previewValue=<1 0 0 1>
  }
  FunctionOutput(None=Clamp(None=Multiply(A=@n1) {constB=2048}) {clampMode=CMODE_ClampMin, minDefault=8}) {
   description=The output value represents the chosen model's length along a given axis starting at the object's pivot point. The return value can represent values between 8 to 2048 in increments of 8.
   outputName=Rescaled Extent
  }
 ]
}
