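# Material function in the Pivot Painter 2 library category that decodes an axis vector from Pivot Painter 2 texture RGB into a world-space direction.
# It remaps the input from 0..1 to -1..1 with a bias and scale, transforms it from instance space to world space, then normalizes the result.
# Using instance space as the transform source makes the decoded axis follow each instance's transform.
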MaterialFunction /Engine/Functions/Engine_MaterialFunctions02/PivotPainter2/ms_PivotPainter2_DecodeAxisVector
graph {
 properties {
  bExposeToLibrary=true
  description=Transforms Pivot Painter 2.0's local space vector info into world space vectors.
  libraryCategoriesText [Pivot Painter 2]
 }
 nodes [
  n1 = FunctionInput() {
   bUsePreviewValueAsDefault=true
   description=Input RGB vector information from a Pivot Painter 2 texture that output those values.
   inputName=Axis Vector RGB
  }
  n2 = ConstantBiasScale(None=@n1) {bias=-0.5, scale=2}
  FunctionOutput(None=Normalize(VectorInput=Transform(None=@n2) {transformSourceType=TRANSFORMSOURCE_Instance})) {description=The input axis vector information has now been transformed to world space.}
 ]
}
