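# Material function in the Pivot Painter 2 library that decodes stored pivot positions into world space.
# It takes the RGB of a texture holding 16-bit Pivot Position data, with the preview value as default.
# A single TransformPosition node converts that value from instance space to world space, yielding each model's captured pivot location.
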MaterialFunction /Engine/Functions/Engine_MaterialFunctions02/PivotPainter2/ms_PivotPainter2_DecodePostion
graph {
 properties {
  bExposeToLibrary=true
  description=Transforms Pivot Painter 2.0's local space position information into world position info.
  libraryCategoriesText [Pivot Painter 2]
 }
 nodes [
  n1 = FunctionInput() {
   bUsePreviewValueAsDefault=true
   description="Insert the RGB values of a texture that contain Pivot Painter 2 “Pivot Position (16-bit)” data."
   inputName=Position RGB
  }
  FunctionOutput(None=TransformPosition(None=@n1) {transformSourceType=TRANSFORMPOSSOURCE_Instance, transformType=TRANSFORMPOSSOURCE_World}) {description="The output value is the world space location of each model’s pivot point position as it was captured by Pivot Painter."}
 ]
}
