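# Pivot Painter 2 material function that locates a parent sub-object's texture data from its parent index.
# The parent index, expected already decoded to float, and the texture dimensions feed 2dArrayLookupByIndex to output the parent's UVs, without sampling any texture.
# A second output compares the parent index with an optionally supplied current index, returning 0 when equal and 1 otherwise to flag child objects.
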MaterialFunction /Engine/Functions/Engine_MaterialFunctions02/PivotPainter2/ms_PivotPainter2_ReturnParentTextureInfo
graph {
 properties {
  bExposeToLibrary=true
  description=Read a parent sub object's texture data using Pivot Painter 2.0 parent index integer as float texture data.
  libraryCategoriesText [Pivot Painter 2]
 }
 nodes [
  n1 = FunctionInput() {
   bUsePreviewValueAsDefault=true
   description="This input assumes that the data is in FLOAT form. If you're reading from a parent index int as float texture please first decode the asset using MaterialFunction'/Engine/Functions/Engine_MaterialFunctions02/ms_PivotPainter2_UnpackIntegerAsFloat'. "
   inputName="Parent Index As Float (See note)"
   inputType=FunctionInput_Scalar
   sortPriority=1
   previewValue=<1 0 0 1>
  }
  n2 = FunctionInput() {
   bUsePreviewValueAsDefault=true
   description="Providing this value is only nessesary if one would like to determine if this asset is a child of another component. "
   inputName="Current Index (see note)"
   sortPriority=5
  }
  n3 = Constant() {r=1}
  n4 = Constant()
  FunctionOutput(None=ComponentMask(None=If(A=@n1, B=@n2, "A > B"=@n3, "A == B"=@n4, "A < B"=@n3)) {r=true}) {
   description=Returns 1 if the object is a child of another object. Returns 0 otherwise. This requires that the current index be entered in the current index input. The current index can be found using the ms_PivotPainter2_CalculateMeshElementIndex material function if you are referencing your textures using the models uvs.
   outputName="Is Child? (See note)"
   sortPriority=5
  }
  n5 = FunctionInput() {bUsePreviewValueAsDefault=true, inputName=Texture Dimensions, inputType=FunctionInput_Vector2, sortPriority=3, previewValue=<2 2 0 1>}
  FunctionOutput(None=MaterialFunctionCall!(Index=@n1, Array Dimensions=@n5) {materialFunction=&/2dArrayLookupByIndex}.UVs) {description=Returns the UV coordinates for the parent object's data location., outputName=Parent UVs, sortPriority=2}
 ]
}
