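# Material function that animates Pivot Painter 2 foliage in wind, outputting world position offset, rotated normals and parent UVs for hierarchy.
# It decodes each element's pivot, X axis and extent from textures, then samples a scrolling wind texture at two scales in wind-aligned space.
# Turbulence perturbs the axis-cross-wind rotation axis, while gust magnitude drives the angle.
# An optional wind-facing shelter mask and distance-from-pivot falloff attenuate rotation, and a layer mask gates it.
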MaterialFunction /Engine/Functions/Engine_MaterialFunctions02/PivotPainter2/PivotPainter2FoliageAnimation
graph {
 properties {libraryCategoriesText [Fortnite]}
 nodes [
  n1 = FunctionInput(Preview=VertexNormalWS()) {bUsePreviewValueAsDefault=true, inputName=Optional - world space pixel normal, previewValue=<0 0 1 1>}
  n2 = FunctionInput() {
   bUsePreviewValueAsDefault=true
   inputName=Random Rotation Influence
   inputType=FunctionInput_Scalar
   sortPriority=12
   previewValue=<0.25 0 0 1>
  }
  n3 = FunctionInput() {
   bUsePreviewValueAsDefault=true
   inputName=Random Rotation Texture Sample Scale
   inputType=FunctionInput_Scalar
   sortPriority=11
   previewValue=<250 0 0 1>
  }
  n4 = FunctionInput(Preview=ScalarParameter() {defaultValue=0.25, group=Shared Wind Settings, parameterName=Wind Speed}) {bUsePreviewValueAsDefault=true, inputName=Wind Speed Down Wind Vector, inputType=FunctionInput_Scalar, previewValue=<1 0 0 1>}
  n5 = FunctionInput(Preview=ScalarParameter() {defaultValue=1, group=Shared Wind Settings, parameterName="Wind Horizontal Speed "}) {
   bUsePreviewValueAsDefault=true
   inputName=Wind Speed Horizontal Speed
   inputType=FunctionInput_Scalar
   sortPriority=51
   previewValue=<1 0 0 1>
  }
  n6 = StaticSwitchParameter(
   True=PreviousFrameSwitch(
    Current Frame=ScalarParameter() {group=Shared Wind Settings, parameterName=CurrentTime}
    Previous Frame=ScalarParameter() {group=Shared Wind Settings, parameterName=PreviousTime}
   )
   False=ScalarParameter() {group=Shared Wind Settings, parameterName=CurrentTime}
  ) {group=Shared Wind Settings, parameterName=Expose Previous Frame Time}
  n7 = Time() {bOverride_Period=true, period=99999997952}
  n8 = FunctionInput(Preview=ScalarParameter() {defaultValue=1, group=Shared Wind Settings, parameterName=Wind Gust World Scale}) {bUsePreviewValueAsDefault=true, inputName=Wind Gust World Scale, inputType=FunctionInput_Scalar, sortPriority=50, previewValue=<1 0 0 1>}
  n9 = FunctionInput(Preview=TextureCoordinate() {coordinateIndex=1}) {
   bUsePreviewValueAsDefault=true
   description=Defaults To UV 1.
   inputName=Pivot Painter Texture Coordinate
   inputType=FunctionInput_Vector2
   sortPriority=-1
  }
  n10 = TextureObjectParameter() {
   group=Shared Wind Settings
   parameterName=Position and Index Texture
   samplerType=SAMPLERTYPE_LinearColor
   texture=/Engine/Functions/Engine_MaterialFunctions02/ExampleContent/PivotPainter2/SimpExPivPos
  }
  n11 = TextureSample(
   UVs=@n9
   Tex=FunctionInput(Preview=@n10) {bUsePreviewValueAsDefault=true, inputName=Position And Index Texture, inputType=FunctionInput_Texture2D, sortPriority=50}
  ) {
   samplerType=SAMPLERTYPE_LinearColor
   texture=/Engine/Functions/Engine_MaterialFunctions02/ExampleContent/PivotPainter2/test2_rgb_PivotPos_a_ParentIndexInt_UV_2
  }
  n12 = MaterialFunctionCall!(Position RGB=@n11.RGB) {materialFunction=&/ms_PivotPainter2_DecodePostion}
  n13 = TextureObjectParameter() {
   group=Shared Wind Settings
   parameterName=X-Vector And X-Extent Texture
   samplerType=SAMPLERTYPE_LinearColor
   texture=/Engine/Functions/Engine_MaterialFunctions02/ExampleContent/PivotPainter2/PP2_rgb_XVect_a_XExtent
  }
  n14 = TextureSample(
   UVs=@n9
   Tex=FunctionInput(Preview=@n13) {bUsePreviewValueAsDefault=true, inputName=X-axis and X-Extent Texture, inputType=FunctionInput_Texture2D, sortPriority=50}
  ) {samplerType=SAMPLERTYPE_LinearColor}
  n15 = MaterialFunctionCall!(Axis Vector RGB=@n14.RGB) {materialFunction=&/ms_PivotPainter2_DecodeAxisVector}
  n16 = MaterialFunctionCall!(8 Bit Alpha Extent Value=@n14.A) {materialFunction=&/ms_PivotPainter2_Decode8BitAlphaAxisExtent}
  n17 = Add(A=@n12.Result, B=Multiply(A=@n15.Result, B=@n16.Rescaled Extent))
  n18 = VectorParameter() {group=Shared Wind Settings, parameterName=Normalized Wind Direction X Axis, defaultValue=<rgba 1 0 0 1>}
  n19 = FunctionInput(Preview=@n18.RGB) {bUsePreviewValueAsDefault=true, inputName=Wind Direction, sortPriority=50, previewValue=<1 0 0 1>}
  n20 = DotProduct(A=@n19, B=@n17)
  n21 = VectorParameter() {
   desc=Note: the X and Y Wind axis should be perpendicular and normalized
   group=Shared Wind Settings
   parameterName=Normalized Wind Direction Y Axis
   defaultValue=<rgba 0 1 0 1>
  }
  n22 = FunctionInput(Preview=@n21.RGB) {bUsePreviewValueAsDefault=true, inputName=Wind Direction Y Axis, sortPriority=50, previewValue=<0 1 0 1>}
  n23 = DotProduct(A=@n22, B=@n17)
  n24 = Add(
   A=Multiply(
    A=StaticSwitchParameter(True=@n6, False=@n7) {group=Shared Wind Settings, parameterName=Manually Control Pivot Painter 2 Foliage Animation Time}
    B=AppendVector(A=@n4, B=@n5)
   )
   B=Divide(A=AppendVector(A=@n20, B=@n23), B=@n8)
  )
  n25 = TextureObjectParameter() {
   group=Shared Wind Settings
   parameterName=WindTurbulenceVectorAndGustMagnitude
   samplerType=SAMPLERTYPE_LinearColor
   texture=/Engine/Functions/Engine_MaterialFunctions02/ExampleContent/Textures/WindTurblenceVectorAndGustMagnitude
  }
  n26 = FunctionInput(Preview=@n25) {bUsePreviewValueAsDefault=true, inputName=Optional - Wind Turbulence and Gust Mag, inputType=FunctionInput_Texture2D, sortPriority=99}
  n27 = TextureSample(UVs=Frac(None=Divide(A=@n24, B=@n3)), Tex=@n26) {
   samplerType=SAMPLERTYPE_LinearColor
   texture=&WindTurblenceVectorAndGustMagnitude
  }
  n28 = Multiply(A=Power(Base=DotProduct(A=@n19, B=@n15.Result)) {constExponent=5}, B=Constant3Vector() {constant=<rgba 0 0 -0.2 0>})
  n29 = Add(A=Multiply(A=@n2, B=Subtract(A=@n27.RGB) {constB=0.5}) {constB=0}, B=CrossProduct(A=@n15.Result, B=Add(A=@n28, B=@n19)))
  n30 = Divide(A=@n29, B=Distance(A=@n29, B=Constant()))
  n31 = FunctionInput() {bUsePreviewValueAsDefault=true, inputName=Advanced - Parent Rotation Angle Influence, inputType=FunctionInput_Scalar, sortPriority=10}
  n32 = FunctionInput() {bUsePreviewValueAsDefault=true, inputName=Wind Gust Angle Rotation, inputType=FunctionInput_Scalar, previewValue=<0.1 0 0 1>}
  n33 = FunctionInput() {bUsePreviewValueAsDefault=true, inputName=Wind Gust Offset, inputType=FunctionInput_Scalar, previewValue=<-0.125 0 0 1>}
  n34 = TextureSample(UVs=Frac(None=@n24), Tex=@n26) {
   samplerType=SAMPLERTYPE_LinearColor
   texture=&WindTurblenceVectorAndGustMagnitude
  }
  n35 = Add(A=@n31, B=Multiply(A=@n32, B=Add(A=@n33, B=@n34.A) {constB=-0.25}))
  n36 = DotProduct(A=@n19, B=@n15.Result)
  n37 = FunctionInput() {
   bUsePreviewValueAsDefault=true
   description="Bias and scale to the results of a dot product between the wind vector and the element axis vector. The defaults make sure that the wind fully affects branches facing the wind while leaving the other branches unaffected. "
   inputName=Wind Shelter Settings
   inputType=FunctionInput_Vector2
   sortPriority=10
   previewValue=<-1 -0.5 0 1>
  }
  n38 = MaterialFunctionCall(Float2=@n37) {materialFunction=&BreakOutFloat2Components}
  n39 = Multiply(A=@n35, B=Multiply(A=Add(A=@n36, B=@n38.R), B=@n38.G)) {desc=Anim * Directional Wind Mask Settings}
  n40 = FunctionInput(Preview=StaticBool() {value=true}) {bUsePreviewValueAsDefault=true, inputName=Use Wind Shelter Settings, inputType=FunctionInput_StaticBool, sortPriority=100}
  n41 = WorldPosition() {worldPositionShaderOffset=WPT_ExcludeAllShaderOffsets}
  n42 = DotProduct(
   A=Divide(
    A=Subtract(A=@n41.XYZ, B=@n12.Result)
    B=MaterialFunctionCall() {materialFunction=/Engine/Functions/Engine_MaterialFunctions02/WorldPositionOffset/ObjectScale}.Scale XYZ
   )
   B=Normalize(VectorInput=@n15.Result)
  )
  n43 = FunctionInput() {bUsePreviewValueAsDefault=true, inputName=Motion Dampening Falloff Radius, inputType=FunctionInput_Scalar, previewValue=<0.25 0 0 1>}
  n44 = Multiply(A=@n43, B=@n16.Rescaled Extent)
  n45 = Multiply(
   A=FunctionOutput(None=StaticSwitch(True=@n39, False=@n35, Value=@n40)) {outputName=Advanced - Rotation Angle Animation, sortPriority=50}
   B=Clamp(None=Divide(A=@n42, B=@n44))
  ) {desc=Distance Mask}
  n46 = FunctionInput() {
   bUsePreviewValueAsDefault=true
   description=Enter a black and white value. Grayscale values will produce incorrect results.
   inputName=Layer Mask
   inputType=FunctionInput_Scalar
   sortPriority=9
   previewValue=<1 0 0 1>
  }
  n47 = LinearInterpolate(
   A=@n1
   B=MaterialFunctionCall(Rotation Axis=@n30, Rotation Angle=@n45, World Space Vertex Normals=@n1) {materialFunction=&FixRotateAboutAxisNormals}.New World Space Vertex Normal
   Alpha=@n46
  )
  n48 = FunctionInput(Preview=StaticBool() {value=true}) {
   bUsePreviewValueAsDefault=true
   description="Defaults to true. "
   inputName=Enable Normal Rotation
   inputType=FunctionInput_StaticBool
   sortPriority=-1
  }
  n49 = FunctionInput(Preview=StaticBool() {value=true}) {
   bUsePreviewValueAsDefault=true
   description="Defaults to true and outputs the correct World Position Offset and Rotated Normal. "
   inputName=Enable
   inputType=FunctionInput_StaticBool
   sortPriority=-2
  }
  FunctionOutput(None=StaticSwitch(True=StaticSwitch(True=@n47, False=@n1, Value=@n48), False=@n1, Value=@n49)) {outputName=Rotated World Space Pixel Normal, sortPriority=10}
  n50 = FunctionOutput(None=@n11.A) {outputName=Parent Index Integer}
  n51 = FunctionInput() {
   bUsePreviewValueAsDefault=true
   description="Use the \"Texture Property\" node with the texture size option selected from the drop down."
   inputName=Texture Dimensions To Gather Parent UVs
   inputType=FunctionInput_Vector2
  }
  FunctionOutput(None=MaterialFunctionCall!(
    "Parent Index As Float (See note)"=MaterialFunctionCall!(Integer as Float=@n50) {materialFunction=/Engine/Functions/Engine_MaterialFunctions02/ms_PivotPainter2_UnpackIntegerAsFloat}.Result
    Texture Dimensions=@n51
    "Current Index (see note)"
   ) {materialFunction=&/ms_PivotPainter2_ReturnParentTextureInfo}.Parent UVs) {outputName=Parent UVs}
  n52 = WorldPosition() {worldPositionShaderOffset=WPT_ExcludeAllShaderOffsets}
  n53 = StaticSwitch(
   True=Multiply(A=RotateAboutAxis(NormalizedRotationAxis=@n30, RotationAngle=@n45, PivotPoint=@n12.Result, Position=@n52.XYZ), B=@n46)
   False=Constant()
   Value=@n49
  )
  FunctionOutput(None=Add(
    A=FunctionOutput(None=@n53) {outputName=World Position Offset}
    B=WorldPosition() {worldPositionShaderOffset=WPT_ExcludeAllShaderOffsets}.XYZ
   )) {outputName=Final World Position, sortPriority=10}
 ]
}
