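# Pivot Painter 2 library material function that derives a model element's ID from the mesh's UVs.
# It multiplies the Pivot Painter UVs (previewed from UV channel 1) by the data texture dimensions and floors the result to get integer texel coordinates.
# The output Index flattens them row-major as x plus y times texture width.
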MaterialFunction /Engine/Functions/Engine_MaterialFunctions02/PivotPainter2/ms_PivotPainter2_CalculateMeshElementIndex
graph {
 properties {bExposeToLibrary=true, description=Pulls the model elements element ID from the models uvs., libraryCategoriesText [Pivot Painter 2]}
 nodes [
  n1 = FunctionInput(Preview=TextureCoordinate() {coordinateIndex=1}) {inputName=Pivot Painter UV Coordinates, inputType=FunctionInput_Vector2}
  n2 = FunctionInput() {
   description=Use the texture property node to gather the texture's dimensions
   inputName=Data Texture Dimensions
   inputType=FunctionInput_Vector2
   previewValue=<16 2 0 1>
  }
  n3 = MaterialFunctionCall(Float2=Floor(None=Multiply(A=@n1, B=@n2))) {materialFunction=&BreakOutFloat2Components}
  n4 = MaterialFunctionCall(Float2=@n2) {materialFunction=&BreakOutFloat2Components}
  FunctionOutput(None=Add(A=@n3.R, B=Multiply(A=@n3.G, B=@n4.R))) {outputName=Index}
 ]
}
